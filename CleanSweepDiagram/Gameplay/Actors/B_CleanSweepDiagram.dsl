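; Fortnite equippable weapon actor for a camo item (CleanSweepDiagram) that manages HUD key text, cooldown feedback and manual early deactivation.
; On equip it caches the pawn and, if on cooldown, waits for tag removal to fire a local gameplay cue.
; A custom event starts a hotfixable timer, after which, if camo is active, the HUD label switches and a primary press sends a force-deactivate gameplay event.
; Holstering and sound indicators are disabled.
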(event Custom|Allow%20Manual%20Input%20Deactivation%20Event
  (bind _player_pawn (Variables|Default|GetPlayer%20Pawn))
  ;[0 0.9847 1] Comment
  (if (GameplayTags|HasMatchingGameplayTag (Variables|Default|GetPlayer%20Pawn) (Variables|Tags|GetCamo%20Active%20Tag%20to%20Check))
    (CallFunction|SetHudKeys true)
    (bind _asynctaskproxy (Ability|Async|WaitGameplayEventToActor _player_pawn "(TagName=\"Event.Primary.Pressed\")" true)
      (:then
        (Variables|AsyncActions|SetAsync%20Wait%20For%20Primary%20Input%20Pressed%20Action _asynctaskproxy)
        (Ability|Async|EndAction (Variables|AsyncActions|GetAsync%20Wait%20for%20Camo%20Active%20Tag%20Added%20Action)))
      (:EventReceived
        (if (GameplayTags|HasMatchingGameplayTag (Variables|Default|GetPlayer%20Pawn) (Variables|Tags|GetCamo%20Active%20Tag%20to%20Check))
          (Ability|SendGameplayEventtoActor _player_pawn (Variables|Tags|GetForce%20Deativate%20On%20Input%20Event%20Tag) (Utilities|Struct|MakeGameplayEventData (Variables|Tags|GetForce%20Deativate%20On%20Input%20Event%20Tag)))
          (Ability|Async|EndAction (Variables|AsyncActions|GetAsync%20Wait%20For%20Primary%20Input%20Pressed%20Action))
          (CallFunction|SetHudKeys false))))))

(event Custom|WaitForDeactivateInput%20Event
  ;[0 1 0.2829] These event watchers wait for primary input action to trigger exit early / on input
  (if (not (GameplayTags|HasMatchingGameplayTag (Variables|Default|GetPlayer%20Pawn) (Variables|Tags|GetCamo%20Active%20Tag%20to%20Check)))
    (bind _asynctaskproxy (Ability|Tasks|WaitGameplayTagAddToActor (Variables|Default|GetPlayer%20Pawn) (Variables|Tags|GetCamo%20Active%20Tag%20to%20Check) true)
      (:then
        (Variables|AsyncActions|SetAsync%20Wait%20for%20Camo%20Active%20Tag%20Added%20Action _asynctaskproxy))
      (:Added
        _)))
  (bind _returnvalue (Utilities|Time|SetTimerbyEvent (EventDispatchers|CreateEvent :!event Custom|Allow%20Manual%20Input%20Deactivation%20Event) (ScalableFloat|GetValueatLevel (Variables|Hotfix|GetSF_TimeBeforeManualDeactivateAllowed))))
  (Variables|Default|SetTime%20Before%20Input%20Deactivation%20Event%20Allowed%20Timer%20Handle _returnvalue))

(event FortWeapon|EventOnEquippedAndAttached
  (bind _returnvalue (Game|GetInstigator))
  ;[0 0.2075 1] On Equip, check if we're on cooldown, if so, wait for cooldown tag to be removed and play GC
  (Utilities|IsValid _returnvalue
    (:Is%20Valid
      (bind _asfort_player_pawn (Utilities|Casting|CastToFortPlayerPawn _returnvalue))
      (Variables|Default|SetPlayer%20Pawn _asfort_player_pawn)
      (CallFunction|SetHudKeys false)
      (bind _returnvalue_2 (GameplayTags|HasMatchingGameplayTag(Message) _returnvalue (Variables|Tags|GetCooldown%20Tag)))
      (if _returnvalue_2
        (bind _asynctaskproxy (Ability|Tasks|WaitGameplayTagRemoveFromActor _returnvalue (Variables|Tags|GetCooldown%20Tag) true)
          (:then
            (Variables|AsyncActions|SetAsync%20Action _asynctaskproxy))
          (:Removed
            (GameplayCue|ExecuteGameplayCue%28Burst%29Local (Game|GetInstigator) (Variables|Tags|GetGameplay%20Cue%20Tag) (Ability|GameplayCue|MakeGameplayCueParameters 0.0))))))))

(event FortWeapon|EventOnUnEquippedAndDetached
  ;[0 1 0.9926] End Async actions if no longer valid equipped item
  (Ability|Async|EndAction (Variables|AsyncActions|GetAsync%20Action)))

(event EventBeginPlay
  (|CallParentFunction) ; This node is disabled and will not be called.\nDrag off pins to build functionality.
)

(event EventTick
  (params (DeltaSeconds :type Float%20%28single-precision%29))
  (|CallParentFunction DeltaSeconds) ; This node is disabled and will not be called.\nDrag off pins to build functionality.
)

(fn CanApplyHolster
  (params (HolsterId :type Name%20%28by%20ref%29 :default "None"))
  (return false))

(fn SetHudKeys
  (params (Camo%20Active :type Boolean))
  (WeaponHUDData|SetHudKeyActionDescription 0 (select Camo%20Active (Variables|HUDDesc|GetManual%20Deactivate%20Allowed%20Action%20Description) (Variables|HUDDesc|GetDefault%20Action%20Description)))
  (return))

(fn ShouldShowSoundIndicator
  (params (PlayerController :type Fort%20Player%20Controller%20Object%20Reference))
  (return false))

(fn ConstructionScript
  (|CallParentFunction))
